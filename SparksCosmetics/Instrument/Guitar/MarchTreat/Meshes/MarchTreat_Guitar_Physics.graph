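# Physics asset for the MarchTreat guitar cosmetic, set up for front and rear dangling ribbons.
# Each ribbon is a three-bone chain of small capsule bodies joined by narrow swing and very tight twist limits, hanging from a shared ribbon_base.
# The ribbon_base is angularly locked to the handle, whose large capsule and small guitar-knob capsule give rough collision shapes for the instrument.
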PhysicsAsset /SparksCosmetics/Instrument/Guitar/MarchTreat/Meshes/MarchTreat_Guitar_Physics
bodies {
 dyn_ribbon_front_01 [{shapeType=Capsule, length=0.5, radius=0.7, shapeName=dyn_ribbon_front_01_capsule, center=<0.2 0 0>, rotation=<rot 90 0 0>}]
 dyn_ribbon_front_02 [{shapeType=Capsule, length=0.5, radius=0.7, shapeName=dyn_ribbon_front_02_capsule, center=<0.2 0 0>, rotation=<rot 90 0 0>}]
 dyn_ribbon_front_03 [{shapeType=Capsule, length=0.5, radius=0.7, shapeName=dyn_ribbon_front_03_capsule, center=<0.2 0 0>, rotation=<rot 90 0 0>}]
 dyn_ribbon_rear_01 [{shapeType=Capsule, length=0.5, radius=0.7, shapeName=dyn_ribbon_rear_01_capsule, center=<0.2 0 0>, rotation=<rot 90 0 0>}]
 dyn_ribbon_rear_02 [{shapeType=Capsule, length=0.5, radius=0.7, shapeName=dyn_ribbon_rear_02_capsule, center=<0.2 0 0>, rotation=<rot 90 0 0>}]
 dyn_ribbon_rear_03 [{shapeType=Capsule, length=0.5, radius=0.7, shapeName=dyn_ribbon_rear_03_capsule, center=<0.2 0 0>, rotation=<rot 90 0 0>}]
 handle [
  {
   shapeType=Capsule
   length=1.5019
   radius=0.7772
   shapeName=guitar knob
   center=<-1.8652 4.1069 66.3338>
   rotation=<rot -27.38 1.1273 -8.1948>
  }
  {shapeType=Capsule, length=15, radius=15, shapeName=handle_capsule, center=<0 0 -7.4202>}
 ]
 ribbon_base [{
   shapeType=Capsule
   length=0.505
   radius=0.505
   shapeName=ribbon_base_capsule
   center=<0.0109 -0.0517 -0.0051>
   rotation=<rot -1.5173 -90 0.946>
  }]
}
constraints [
 {bone1Name=dyn_ribbon_rear_03, bone2Name=dyn_ribbon_rear_02, swing1LimitDegrees=20, swing2LimitDegrees=10, twistLimitDegrees=4}
 {bone1Name=dyn_ribbon_rear_01, bone2Name=ribbon_base, swing1LimitDegrees=15, swing2LimitDegrees=10, twistLimitDegrees=2}
 {bone1Name=dyn_ribbon_rear_02, bone2Name=dyn_ribbon_rear_01, swing1LimitDegrees=25, swing2LimitDegrees=10, twistLimitDegrees=2}
 {bone1Name=dyn_ribbon_front_03, bone2Name=dyn_ribbon_front_02, swing1LimitDegrees=20, swing2LimitDegrees=10, twistLimitDegrees=4}
 {bone1Name=dyn_ribbon_front_02, bone2Name=dyn_ribbon_front_01, swing1LimitDegrees=25, swing2LimitDegrees=10, twistLimitDegrees=2}
 {bone1Name=dyn_ribbon_front_01, bone2Name=ribbon_base, swing1LimitDegrees=15, swing2LimitDegrees=15, twistLimitDegrees=2}
 {
  bone1Name=ribbon_base
  bone2Name=handle
  swing1LimitDegrees=45
  swing1Motion=Locked
  swing2LimitDegrees=45
  swing2Motion=Locked
  twistLimitDegrees=45
  twistMotion=Locked
 }
]
